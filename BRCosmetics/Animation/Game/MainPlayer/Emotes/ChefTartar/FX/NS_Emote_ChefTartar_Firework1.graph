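# Niagara system for the ChefTartar emote's fireworks, launching randomly colored flares in two mirrored spawn groups.
# Parent flares feed ribbon smoke and trail embers through particle attribute reads.
# Scratch modules pass flare death position and color through system arrays to spawn burst mesh, flash with light, smoke and sparks.
# Particles shift world/local blend with owner speed, some brightness scales with camera distance, and audio modules are disabled.
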NiagaraSystem /BRCosmetics/Animation/Game/MainPlayer/Emotes/ChefTartar/FX/NS_Emote_ChefTartar_Firework1
dependencies {
 usedDataInterfaces [
  &NiagaraDataInterfaceParticleRead
  &NiagaraDataInterfaceCurve
  &NiagaraDataInterfaceColorCurve
  &NiagaraDataInterfaceArrayColor
  &NiagaraDataInterfaceAudioPlayer
  &NiagaraDataInterfaceCamera
  &NiagaraDataInterfaceArrayPosition
 ]
 usedDynamicInputs [&MakeVector2D]
 usedModules [
  &self:SystemSpawnScript.SystemScriptSource.SystemScriptGraph.NiagaraNodeAssignment_7.SetVariables_#1
  &SystemState
  &BlendSpaceHelper
  &self:ReadDeath_02
  &self:SystemSpawnScript.SystemScriptSource.SystemScriptGraph.NiagaraNodeAssignment_8.SetVariables_#2
 ]
}
emitters {
 NE_FireworkFlare_Parent {
  emitterUpdateScript [
   &EmitterState()
   &SpawnBurst_Instantaneous(Use Spawn Probability=false, UseLoopCountLimit=false)
   &SpawnBurst_Instantaneous(Spawn Group=1, Spawn Time=0.05, Use Spawn Probability=false, UseLoopCountLimit=false) {module=SpawnBurst_Instantaneous001}
   &SpawnBurst_Instantaneous(Spawn Time=0.6, Use Spawn Probability=false, UseLoopCountLimit=false) {module=SpawnBurst_Instantaneous002}
   &SpawnBurst_Instantaneous(Spawn Group=1, Spawn Time=0.65, Use Spawn Probability=false, UseLoopCountLimit=false) {module=SpawnBurst_Instantaneous003}
   &SpawnBurst_Instantaneous(Spawn Time=1, Use Spawn Probability=false, UseLoopCountLimit=false) {module=SpawnBurst_Instantaneous004}
   &SpawnRate(SpawnRate=2, Use Spawn Probability=false)
  ]
  particleSpawnScript [
   &FN_SetSpawnGroup(Number of Spawn Groups=2)
   &InitializeParticle(
    Color=&SelectLinearColorFromArray(
     Array Sampling Mode=::NewEnumerator0
     Direct Array Index=?
     Color Selection Array=DataInterface() {propertyValues {ColorData [
        <rgba 2 2 2 1>
        <rgba 1.9113 2 0.0006 1>
        <rgba 0.0781 0.0781 3 1>
        <rgba 0 2 0 1>
        <rgba 2 0.0012 1.9294 1>
        <rgba 3 0.0938 0.0947 1>
        <rgba 2 0.119 0 1>
       ]}}
     Random Seed=0
    )
    Lifetime Max=0.7
    Lifetime Min=0.5
    Lifetime Mode=::NewEnumerator1
    Sprite Rotation Mode=::NewEnumerator2
    Sprite Size Mode=::NewEnumerator1
    Uniform Sprite Size Max=45
    Uniform Sprite Size Min=40
    UsePositionOffset=false
    Write Material Random=true
   )
   &ShapeLocation(
    Box / Plane Mode:ENiagara_BoxPlaneMode=::NewEnumerator1
    MaskSpawnGroup=false
    Override Seed=false
    Plane Midpoint:Vector2f=&MaskVector2dBySpawnGroup(Masked Vector2D=<-0.6 0>, Passthrough Vector2D=<1.6 0>, Spawn Group Mask=0)
    Plane Size:Vector2f=<18 -5>
    Shape Primitive=::NewEnumerator2
    UseOwnerScale=true
   ) {module=ShapeLocation001}
   /SetVariables(Particles.Position2:NiagaraPosition=&AddVectorToPosition(
     Position=$Particles.ShapeLocation001.ShapePosition
     Vector=&TransformVector(
      Vector=&MakeVector(X=$User.Float, Y=0, Z=&FN_FrontEndFloatModifier(Gameplay=0, Front End=3))
      Destination Space=::World
     )
    )) {isSetParameters=true}
   &AddVelocity(
    Module.Use Falloff Distance=false
    Origin Offset=&MaskVectorBySpawnGroup(
     Masked Vector=&RandomRangeVector(Minimum=<-20 0 -90>, Maximum=<-20 0 -100>)
     Passthrough Vector=&RandomRangeVector(Minimum=<20 0 -90>, Maximum=<20 0 -100>)
    )
    Override Seed=false
    Velocity Mode=::NewEnumerator1
    Velocity Speed=&RandomRangeFloat(Minimum=700, Maximum=1000)
   ) {module=AddVelocity001}
   /SetVariables(Particles.Position=&AddVectorToPosition(
     Position=$Particles.Position
     Vector=&TransformVector(Vector=&MakeVector(X=$User.Float, Y=0, Z=0), Destination Space=::World)
    )) {isSetParameters=true}
   /Niagara/Modules/Audio/PlayAudio(
    Audio Player=DataInterface() {propertyValues {
      Attenuation=/Game/Sounds/Attenuation/Fort_Emotes_Athena_LessOcclusion_Att
      Concurrency=null
      MaxPlaysPerTick=10
      SoundToPlay=/BRCosmetics/Sounds/Emotes/ChefTartar/IP_MSS_Emote_Foley_ChefTartar_Firework_Launch
      bAllowLoopingOneShotSounds=false
      bLimitPlaysPerTick=true
      bOnlyActiveDuringGameplay=false
      bStopWhenComponentIsDestroyed=true
      ConfigurationUserParameter {parameter {name=None, typeDefHandle {registeredTypeIndex=1}, varData []}}
      ParameterNames []
     }}
    PlayAudio=true
   ) {enabled=false}
  ]
  particleUpdateScript [
   &BlendSpace("Position Blend (World <-> Local)":NiagaraFloat=&MakeCustomFloatFromBool(
     Bool=&SetBoolByFloatComparison(A=&VectorLength(Vector=$Engine.Owner.Velocity), B=$User.Vel)
     True Float=1
     False Float=0.3
     Translation Contribution:NiagaraFloat=1
     Rotation Contribution:NiagaraFloat=0
    ))
   &ParticleState()
   &Drag(Drag=3, Use Linear Drag=true, Use Rotational Drag=false)
   /SetVariables(Particles.SpriteSize2=&Vector2DFromFloat(Value=&FloatFromCurve(
      FloatCurve=DataInterface() {propertyValues {Curve {keys [<at 0 0>, <at 0.0213 1>, {arriveTangent=-0.0405, time=0.1047, value=1}, {arriveTangent=-8.0089, leaveTangent=-8.0089, time=0.17, value=0}]}}}
      Scale Curve=&FN_FrontEndFloatModifier(
       Gameplay=&RandomRangeFloat(Minimum=50, Maximum=60)
       Front End=&RandomRangeFloat(Minimum=50, Maximum=60)
      )
     ))) {isSetParameters=true}
   &ScaleSpriteSize(Scale Factor=&Vector2DFromFloat(Value=&FloatFromCurve(FloatCurve=DataInterface() {propertyValues {Curve {keys [
          {arriveTangent=5.0716, interpMode=RCIM_Cubic, leaveTangent=5.0716, tangentMode=RCTM_User, time=0, value=0}
          {arriveTangent=0.7499, interpMode=RCIM_Cubic, leaveTangent=0.7499, time=0.1078, value=0.2796}
          <at 0.6168 0.4626>
          <at 0.6485 0.6638>
          <at 0.7101 0.2796>
          <at 0.7567 0.5723>
          <at 0.8108 0.2887>
          <at 0.8574 0.8011>
          <at 0.8947 0.2887>
          <at 0.932 1>
          <at 0.95 0.002>
         ]}}})))
   &SolveForcesAndVelocity(Clamp Velocity=false, Limit Acceleration=false)
   &self:WriteDeath_02(Color Array:NiagaraDataInterfaceArrayColor=$System.ColorArray, Position Array:NiagaraDataInterfaceArrayPosition=$System.DeathArray) {module=WriteDeath_02}
   /SetVariables(Particles.Color2=&Multiply_LinearColorByFloat(
     Linear Color=&Multiply_LinearColorByFloat(Linear Color=$Particles.Color, Float=1)
     Float=&FloatFromCurve(FloatCurve=DataInterface() {propertyValues {Curve {keys [
          <at 0 0>
          {tangentMode=RCTM_User, time=0.0293, value=1}
          {arriveTangent=-0.0405, leaveTangent=-0.0405, time=0.1369, value=1}
          {arriveTangent=-1.0085, leaveTangent=-1.0085, time=0.17, value=0}
         ]}}})
    )) {isSetParameters=true}
   &FN_Light_Attributes(
    Color=$Particles.Color
    Create Color Attribute=true
    Create Exponent Attribute=true
    Create Position Attribute=true
    Create Radius Attribute=true
    Exponent:NiagaraFloat=&FN_FrontEndFloatModifier(Gameplay=4, Front End=4)
    Position=&AddVectorToPosition(Position=$Particles.Position2, Vector=<0 0 2>)
    Radius=&FN_FrontEndFloatModifier(Gameplay=100, Front End=100)
    Scale Color=&FloatFromCurve(
     FloatCurve=DataInterface() {propertyValues {Curve {defaultValue=1, keys [{tangentMode=RCTM_User, time=0, value=1}, {tangentMode=RCTM_User, time=0.3, value=0}]}}}
     Scale Curve=&FN_FrontEndFloatModifier(Gameplay=3, Front End=0)
    )
   )
   &ScaleColor(
    Scale RGB=&VectorFromFloat(Value=&FloatFromCurve(FloatCurve=DataInterface() {propertyValues {Curve {
         defaultValue=1
         keys [
          {tangentMode=RCTM_User, time=0, value=1}
          {arriveTangent=-0.0405, interpMode=RCIM_Cubic, leaveTangent=-0.0405, tangentMode=RCTM_User, time=0.7009, value=1}
          {arriveTangent=-8.0089, interpMode=RCIM_Cubic, leaveTangent=-8.0089, tangentMode=RCTM_User, time=1, value=0}
         ]
        }}}))
    ScaleA=false
    ScaleRGB=true
   ) {module=ScaleColor001}
  ]
  rendererClasses [&NiagaraSpriteRendererProperties, &NiagaraLightRendererProperties]
  renderers [
   {rendererClass=&NiagaraSpriteRendererProperties}
   {rendererClass=&NiagaraLightRendererProperties, rendererIndex=1}
   {rendererClass=&NiagaraSpriteRendererProperties, rendererIndex=2}
  ]
 }
 NE_Fireworks_Mesh {
  emitterUpdateScript [
   &EmitterState()
   &SpawnPerFrame(Spawn:NiagaraBool=$System.DeathSpawn, Use Spawn Probability=false)
  ]
  particleSpawnScript [
   &InitializeParticle(
    Color=$System.DeathColor
    Lifetime Max=1
    Lifetime Min=0.5
    Lifetime Mode=::NewEnumerator1
    Mesh Scale Mode=::NewEnumerator1
    Mesh Uniform Scale Max=0.05
    Mesh Uniform Scale Min=0.01
    Position=$System.DeathPosition
    Position Mode=::NewEnumerator1
    UsePositionOffset=false
    Write Material Random=true
   ) {module=InitializeParticle001}
   &InitialMeshOrientation(Use Rotation Vector=false) {module=InitialMeshOrientation001}
  ]
  particleUpdateScript [
   &ParticleState()
   &BlendSpace("Position Blend (World <-> Local)":NiagaraFloat=&MakeCustomFloatFromBool(
     Bool=&SetBoolByFloatComparison(A=&VectorLength(Vector=$Engine.Owner.Velocity), B=$User.Vel)
     True Float=1
     False Float=0.3
     Translation Contribution:NiagaraFloat=1
     Rotation Contribution:NiagaraFloat=0
    ))
   &ScaleColor(
    Scale Alpha=&FloatFromCurve(
     FloatCurve=DataInterface() {propertyValues {Curve {keys [
         {arriveTangent=-2.4774, interpMode=RCIM_Cubic, leaveTangent=-2.4774, tangentMode=RCTM_User, time=0.1389, value=1.3591}
         {arriveTangent=-0.6878, interpMode=RCIM_Cubic, leaveTangent=-0.6878, tangentMode=RCTM_User, time=1, value=0.4}
        ]}}}
     Scale Curve=&RemapRange(
      Input Value To Remap=/Niagara/DynamicInputs/Camera/DistanceToCamera(Use Owner Position:NiagaraBool=true, Camera Query:NiagaraDataInterfaceCamera=DataInterface())
      Input Min=$User.Min
      Input Max=$User.Max
      Output Min=0.5
      Output Max=1
     )
    )
    Scale RGB=&Multiply_VectorByFloat(Vector=<1 1 1>, Float=15)
    ScaleA=true
    ScaleRGB=true
   ) {module=ScaleColor002}
   &ScaleMeshSize(Scale Factor=&VectorFromFloat(Value=&FloatFromCurve(FloatCurve=DataInterface() {propertyValues {Curve {keys [
          {arriveTangent=4.6803, interpMode=RCIM_Cubic, leaveTangent=4.6803, tangentMode=RCTM_User, time=0, value=0}
          {arriveTangent=1.6276, interpMode=RCIM_Cubic, leaveTangent=1.6276, tangentMode=RCTM_User, time=0.1923, value=0.5575}
          {arriveTangent=0.2011, interpMode=RCIM_Cubic, leaveTangent=0.5033, tangentMode=RCTM_User, time=1, value=1}
         ]}}})))
   &GravityForce(Gravity=<0 0 -15>)
   &SolveForcesAndVelocity(Clamp Velocity=false, Limit Acceleration=false)
   &DynamicMaterialParameters(
    Index 0 Param 1=1
    Index 0 Param 2=1
    Index 0 Param 3=&FloatFromCurve(
     FloatCurve=DataInterface() {propertyValues {Curve {keys [
         {arriveTangent=3.0114, interpMode=RCIM_Cubic, leaveTangent=3.0114, tangentMode=RCTM_User, time=0, value=0}
         {arriveTangent=0.9456, interpMode=RCIM_Cubic, leaveTangent=0.9456, tangentMode=RCTM_User, time=0.4343, value=0.7445}
         {arriveTangent=0.3236, interpMode=RCIM_Cubic, leaveTangent=0.7406, tangentMode=RCTM_User, time=1, value=1}
        ]}}}
     Scale Curve=0.2
    )
    Index 0 Param 4=&FloatFromCurve(
     FloatCurve=DataInterface() {propertyValues {Curve {keys [
         {arriveTangent=3.4321, interpMode=RCIM_Cubic, leaveTangent=3.4321, tangentMode=RCTM_User, time=0, value=1}
         {arriveTangent=-1.7957, interpMode=RCIM_Cubic, leaveTangent=-1.7957, tangentMode=RCTM_User, time=0.1672, value=1.5}
         {arriveTangent=-0.286, interpMode=RCIM_Cubic, leaveTangent=-0.286, tangentMode=RCTM_User, time=1, value=0.5}
        ]}}}
     Scale Curve=0.5
    )
    Module.Index0Float0Write=true
    Module.Index0Float1Write=true
    Module.Index0Float2Write=true
    Module.Index0Float3Write=true
   )
  ]
  rendererClasses [&NiagaraMeshRendererProperties]
  renderers [{rendererClass=&NiagaraMeshRendererProperties, rendererIndex=0}]
 }
 NE_Flare {
  emitterUpdateScript [
   &EmitterState()
   &SpawnPerFrame(Spawn:NiagaraBool=$System.DeathSpawn, Use Spawn Probability=false)
  ]
  particleSpawnScript [
   &InitializeParticle(
    Color=$System.DeathColor
    Lifetime Max=0.35
    Lifetime Min=0.2
    Lifetime Mode=::NewEnumerator1
    Position=$System.DeathPosition
    Position Mode=::NewEnumerator1
    Sprite Rotation Mode=::NewEnumerator2
    Sprite Size Mode=::NewEnumerator1
    Uniform Sprite Size Max=150
    Uniform Sprite Size Min=100
    UsePositionOffset=false
    Write Material Random=true
   )
   &PlayAudio(
    Audio Player=DataInterface() {propertyValues {
      Attenuation=&Fort_Emotes_Athena_LessOcclusion_Att
      Concurrency=null
      MaxPlaysPerTick=10
      SoundToPlay=/BRCosmetics/Sounds/Emotes/ChefTartar/IP_MSS_Emote_Foley_ChefTartar_Firework_Explode
      bAllowLoopingOneShotSounds=false
      bLimitPlaysPerTick=true
      bOnlyActiveDuringGameplay=false
      bStopWhenComponentIsDestroyed=true
      ConfigurationUserParameter {parameter {name=None, typeDefHandle {registeredTypeIndex=1}, varData []}}
      ParameterNames []
     }}
    PlayAudio=true
   ) {enabled=false}
  ]
  particleUpdateScript [
   &ParticleState()
   &BlendSpace("Position Blend (World <-> Local)":NiagaraFloat=&MakeCustomFloatFromBool(
     Bool=&SetBoolByFloatComparison(A=&VectorLength(Vector=$Engine.Owner.Velocity), B=$User.Vel)
     True Float=1
     False Float=0.3
     Translation Contribution:NiagaraFloat=1
     Rotation Contribution:NiagaraFloat=0
    ))
   &SolveForcesAndVelocity(Clamp Velocity=false, Limit Acceleration=false)
   &ScaleColor(
    Scale Alpha=&FloatFromCurve(
     FloatCurve=DataInterface() {propertyValues {Curve {keys [
         {arriveTangent=35.674, interpMode=RCIM_Cubic, leaveTangent=35.6741, tangentMode=RCTM_User, time=0, value=0}
         {arriveTangent=1.228, interpMode=RCIM_Cubic, leaveTangent=1.228, time=0.0458, value=1.0493}
         {arriveTangent=-1.0997, interpMode=RCIM_Cubic, leaveTangent=-1.0997, time=0.5705, value=0.7006}
         {interpMode=RCIM_Cubic, leaveTangent=-4.2898, time=1, value=0}
        ]}}}
     Scale Curve=&RemapRange(
      Input Value To Remap=&DistanceToCamera(Use Owner Position:NiagaraBool=true, Camera Query:NiagaraDataInterfaceCamera=DataInterface())
      Input Min=$User.Min
      Input Max=$User.Max
      Output Min=0.1
      Output Max=1
     )
    )
    Scale RGB=<5 5 5>
    ScaleA=true
    ScaleRGB=true
   )
   &ScaleSpriteSize(Scale Factor=&Vector2DFromFloat(Value=&FloatFromCurve(
      FloatCurve=DataInterface() {propertyValues {Curve {keys [<at 0 1.5 RCIM_Cubic>,<at 1 0>]}}}
      Scale Curve=&FN_FrontEndFloatModifier(Gameplay=1, Front End=0.7)
     )))
   &CameraOffset(Camera Offset Amount=50)
   &FN_Light_Attributes(
    Color=$Particles.Initial.Color
    Create Color Attribute=true
    Create Exponent Attribute=true
    Create Radius Attribute=true
    Exponent=4
    Radius=150
    Scale Color=&FloatFromCurve(
     FloatCurve=DataInterface() {propertyValues {Curve {
        defaultValue=1
        keys [
         {tangentMode=RCTM_User, time=0, value=1}
         {arriveTangent=-0.0405, interpMode=RCIM_Cubic, leaveTangent=-0.0405, tangentMode=RCTM_User, time=0.7009, value=1}
         {arriveTangent=-8.0089, interpMode=RCIM_Cubic, leaveTangent=-8.0089, tangentMode=RCTM_User, time=1, value=0}
        ]
       }}}
     Scale Curve=&RemapRange(
      Input Value To Remap=&DistanceToCamera(Use Owner Position:NiagaraBool=true, Camera Query:NiagaraDataInterfaceCamera=DataInterface())
      Input Min=$User.Min
      Input Max=$User.Max
      Output Min=0.5
      Output Max=1
     )
    )
   )
  ]
  rendererClasses [&NiagaraLightRendererProperties, &NiagaraSpriteRendererProperties]
  renderers [
   {rendererClass=&NiagaraLightRendererProperties, rendererIndex=0}
   {rendererClass=&NiagaraSpriteRendererProperties, rendererIndex=1}
  ]
 }
 NE_RibbonSmoke {
  emitterUpdateScript [
   /SetVariables(Emitter.NewSpawnRate=150) {isSetParameters=true}
   &EmitterState(
    Enable Distance Culling=true
    MaxDistance=2500000
    Module.bUseMaxDistance=true
    Module.bUseMinDistance=false
    Scalability Mode=::NewEnumerator1
   )
   &SpawnParticlesFromOtherEmitter(
    Attribute Reader=DataInterface() {propertyValues {EmitterBinding {bindingMode=Other, emitterName=NE_FireworkFlare_Parent}}}
    Calculate Spawn Rate Per Particle=false
    Spawn Rate=$Emitter.NewSpawnRate
    Use Maximum Age=false
    Use Minimum Age=false
   )
  ]
  particleSpawnScript [
   &InitializeRibbon(
    Lifetime=&RandomRangeFloat(Minimum=0.3, Maximum=0.5)
    Position Mode:ENiagara_PositionInitializationMode=::NewEnumerator1
    Ribbon Width=1
    UsePositionOffset=false
    Write Material Random=true
   ) {module=InitializeRibbon001}
   &SampleParticlesFromOtherEmitter(
    Age Sampling:ENiagara_DirectReadSamplingMode=::NewEnumerator0
    Applied RGB Scale:Vector3f=&Multiply_VectorByFloat(Vector=<1 1 1>, Float=1.5)
    Color Sampling=::NewEnumerator0
    Lifetime Application:ENiagara_LifetimeInheritanceOptions=::NewEnumerator1
    Lifetime Sampling:ENiagara_DirectReadSamplingMode=::NewEnumerator0
    Particle ID Sampling=::NewEnumerator1
    Reject Particle If Sampled Index Is Invalid=false
    Sprite Size Sampling:ENiagara_DirectReadSamplingMode=::NewEnumerator0
   )
  ]
  particleUpdateScript [
   &ParticleState()
   &BlendSpace("Position Blend (World <-> Local)":NiagaraFloat=&MakeCustomFloatFromBool(
     Bool=&SetBoolByFloatComparison(A=&VectorLength(Vector=$Engine.Owner.Velocity), B=$User.Vel)
     True Float=1
     False Float=0.3
     Translation Contribution:NiagaraFloat=1
     Rotation Contribution:NiagaraFloat=0
    ))
   /Niagara/Modules/AttributeReader/UpdateParticlesFromOtherEmitter(
    Attribute Reader=DataInterface() {propertyValues {EmitterBinding {bindingMode=Other, emitterName=NE_FireworkFlare_Parent}}}
    Position Sampling:ENiagara_DirectReadSamplingMode=::NewEnumerator1
   )
   &self:ScratchModule(Position:NiagaraPosition=$Output.UpdateParticlesFromOtherEmitter.SampledPosition) {module=ScratchModule}
   &OffsetPosition()
   &ScaleColor(
    Linear Color Curve=DataInterface() {propertyValues {
      AlphaCurve {keys [<at 0 1 RCIM_Cubic>,<at 0.0561 1 RCIM_Cubic>,<at 0.4419 1>,<at 0.823 0 RCIM_Cubic>]}
      BlueCurve {keys [<at 0.05 1>,<at 1 0.9479>]}
      GreenCurve {keys [<at 0.05 1>,<at 1 0.9479>]}
      RedCurve {keys [<at 0.05 1>,<at 1 0.9479>]}
     }}
    Scale Mode=::NewEnumerator2
   )
   &ScaleRibbonWidth(Ribbon Width Scale=&FloatFromCurve(
     FloatCurve=DataInterface() {propertyValues {Curve {keys [
         {arriveTangent=-0.9162, interpMode=RCIM_Cubic, leaveTangent=-0.9162, tangentMode=RCTM_User, time=0.1, value=1}
         {arriveTangent=-0.5263, interpMode=RCIM_Cubic, leaveTangent=-0.5263, time=0.7783, value=0.4618}
         <at 2 0 RCIM_Cubic>
        ]}}}
     Scale Curve=15
    )) {module=ScaleRibbonWidth001}
   /Niagara/Modules/Update/Forces/CurlNoiseForce(
    Noise Frequency=1
    Noise Strength=&Multiply_Float(
     A=250
     B=&FloatFromCurve(
      FloatCurve=DataInterface() {propertyValues {Curve {keys [
          {arriveTangent=1.9891, interpMode=RCIM_Cubic, leaveTangent=1.9891, tangentMode=RCTM_User, time=0, value=0}
          {interpMode=RCIM_Cubic, tangentMode=RCTM_User, time=0.75, value=1}
          {tangentMode=RCTM_User, time=1, value=0}
         ]}}}
      Scale Curve=0.2
     )
    )
    Pan Noise=true
   )
   &SolveForcesAndVelocity(Clamp Velocity=false, Limit Acceleration=false)
   &DynamicMaterialParameters(
    Index 0 Param 1=-1
    Index 0 Param 2=1
    Index 0 Param 4=&FloatFromCurve(FloatCurve=DataInterface() {propertyValues {Curve {keys [
         {arriveTangent=-2.8039, interpMode=RCIM_Cubic, time=0, value=1}
         {arriveTangent=-3.1818, interpMode=RCIM_Cubic, leaveTangent=-3.1818, tangentMode=RCTM_User, time=1, value=0}
        ]}}})
    Module.Index0Float0Write=true
    Module.Index0Float1Write=true
    Module.Index0Float2Write=true
    Module.Index0Float3Write=true
   )
  ]
  rendererClasses [&NiagaraRibbonRendererProperties]
  renderers [{rendererClass=&NiagaraRibbonRendererProperties}]
 }
 NE_Smoke {
  emitterUpdateScript [
   &EmitterState()
   &SpawnPerFrame(Spawn:NiagaraBool=$System.DeathSpawn, Use Spawn Probability=false)
  ]
  particleSpawnScript [
   &InitializeParticle(
    Color=$System.DeathColor
    Lifetime Min=0.7
    Lifetime Mode=::NewEnumerator1
    Position=$System.DeathPosition
    Position Mode=::NewEnumerator1
    Sprite Rotation Mode=::NewEnumerator2
    Sprite Size Mode=::NewEnumerator1
    Uniform Sprite Size Max=120
    Uniform Sprite Size Min=80
    UsePositionOffset=false
    Write Material Random=true
   )
   /SetVariables(Particles.InitSpriteRotRate=&UniformRangedFloat(Minimum=-90, Maximum=90)) {isSetParameters=true}
  ]
  particleUpdateScript [
   &ParticleState()
   &BlendSpace("Position Blend (World <-> Local)":NiagaraFloat=&MakeCustomFloatFromBool(
     Bool=&SetBoolByFloatComparison(A=&VectorLength(Vector=$Engine.Owner.Velocity), B=$User.Vel)
     True Float=1
     False Float=0.3
     Translation Contribution:NiagaraFloat=1
     Rotation Contribution:NiagaraFloat=0
    ))
   &GravityForce(Gravity=<0 0 -35>)
   &SolveForcesAndVelocity(Clamp Velocity=false, Limit Acceleration=false)
   &ScaleColor(
    Scale Alpha=&FloatFromCurve(
     FloatCurve=DataInterface() {propertyValues {Curve {keys [
         {arriveTangent=11.9154, interpMode=RCIM_Cubic, leaveTangent=11.9154, tangentMode=RCTM_User, time=0, value=0}
         {arriveTangent=0.2004, interpMode=RCIM_Cubic, leaveTangent=0.2004, tangentMode=RCTM_User, time=0.1389, value=0.5438}
         {interpMode=RCIM_Cubic, leaveTangent=-3.3562, time=1.0011, value=0.002}
        ]}}}
     Scale Curve=0.4
    )
    Scale RGB=<2 2 2>
    ScaleA=true
    ScaleRGB=true
   )
   &SpriteRotationRate(Rotation Rate=&Multiply_Float(
     A=$Particles.InitSpriteRotRate
     B=&FloatFromCurve(FloatCurve=DataInterface() {propertyValues {Curve {keys [{arriveTangent=-1.3839, interpMode=RCIM_Cubic, leaveTangent=-1.3839, tangentMode=RCTM_User, time=0, value=1}, <at 1 0.1 RCIM_Cubic>]}}})
    ))
   /Niagara/Modules/Update/SubUV/SubUVAnimation(Number Of Frames=64, SubUV Lookup Index=&Multiply_Float(A=$Particles.NormalizedAge, B=0.63))
   &ScaleSpriteSize(Scale Factor=&Vector2DFromFloat(Value=&FloatFromCurve(FloatCurve=DataInterface() {propertyValues {Curve {keys [
          {arriveTangent=4.9, interpMode=RCIM_Cubic, leaveTangent=4.9, tangentMode=RCTM_User, time=0, value=0}
          {arriveTangent=1.1982, interpMode=RCIM_Cubic, leaveTangent=1.1982, tangentMode=RCTM_User, time=0.1791, value=0.4838}
          {arriveTangent=0.2347, interpMode=RCIM_Cubic, leaveTangent=-1.877, tangentMode=RCTM_User, time=1, value=1}
         ]}}})))
   &DynamicMaterialParameters(
    Index 0 Param 1=&FloatFromCurve(FloatCurve=DataInterface() {propertyValues {Curve {keys [<at 0 1>,<at 1 0>]}}})
    Index 0 Param 2=1
    Index 0 Param 3=1
    Index 0 Param 4=1
    Module.Param0WriteEnabled=true
    Module.Param1WriteEnabled=true
    Module.Param2WriteEnabled=true
    Module.Param3WriteEnabled=true
   )
  ]
  rendererClasses [&NiagaraSpriteRendererProperties]
  renderers [{rendererClass=&NiagaraSpriteRendererProperties}]
 }
 NE_Spark {
  emitterUpdateScript [
   &EmitterState()
   &SpawnPerFrame(Spawn:NiagaraBool=$System.DeathSpawn, Spawn Count=150, Use Spawn Probability=false) {module=SpawnPerFrame001}
  ]
  particleSpawnScript [
   &InitializeParticle(
    Color=$System.DeathColor
    Lifetime Min=0.7
    Lifetime Mode=::NewEnumerator1
    Position=$System.DeathPosition
    Position Mode=::NewEnumerator1
    Sprite Size Mode=::NewEnumerator1
    Uniform Sprite Size Max=3
    Uniform Sprite Size Min=2
    UsePositionOffset=false
    Write Material Random=true
   )
   &ShapeLocation(MaskSpawnGroup=false, Override Seed=false, Sphere Radius=22, UseOwnerScale=true)
   &AddVelocity(
    Module.Use Falloff Distance=false
    Override Seed=false
    Velocity Mode=::NewEnumerator1
    Velocity Origin=$System.DeathPosition
    Velocity Speed=&RandomRangeFloat(Minimum=200, Maximum=500)
   )
   &ApplyInitialForces()
  ]
  particleUpdateScript [
   &ParticleState()
   &BlendSpace("Position Blend (World <-> Local)":NiagaraFloat=&MakeCustomFloatFromBool(
     Bool=&SetBoolByFloatComparison(A=&VectorLength(Vector=$Engine.Owner.Velocity), B=$User.Vel)
     True Float=1
     False Float=0.3
     Translation Contribution:NiagaraFloat=1
     Rotation Contribution:NiagaraFloat=0
    ))
   &GravityForce(Gravity=&RandomRangeVector(Minimum=<0 0 -250>, Maximum=<0 0 -250>))
   &Drag(Drag=&RandomRangeFloat(Minimum=6, Maximum=11), Use Linear Drag=true, Use Rotational Drag=true)
   &SolveForcesAndVelocity(Clamp Velocity=false, Limit Acceleration=false)
   &ScaleSpriteSize(Uniform Curve Sprite Scale=DataInterface() {propertyValues {Curve {keys [{tangentMode=RCTM_User, time=0, value=1}, <at 0.3 1>, {tangentMode=RCTM_User, time=1, value=0}]}}})
   &ScaleSpriteSizeBySpeed(Max Scale Factor=<0.2 5>, Min Scale Factor=<1 0.5>)
   &ScaleColor(
    Scale Alpha=&FloatFromCurve(FloatCurve=DataInterface() {propertyValues {Curve {keys [<at 0 1>,<at 0.3 1>,<at 1 0>]}}})
    Scale RGB=&VectorFromFloat(Value=&Multiply_Float(A=1, B=3))
    ScaleA=true
    ScaleRGB=true
   )
  ]
  rendererClasses [&NiagaraSpriteRendererProperties]
  renderers [{rendererClass=&NiagaraSpriteRendererProperties}]
 }
 NE_TrailEmbers {
  emitterUpdateScript [
   &EmitterState()
   &SpawnParticlesFromOtherEmitter(
    Attribute Reader=DataInterface() {propertyValues {EmitterBinding {bindingMode=Other, emitterName=NE_FireworkFlare_Parent}}}
    Spawn Rate=80
    Spawn Rate Per Particle Cap=150
    Use Maximum Age=false
    Use Minimum Age=false
   )
  ]
  particleSpawnScript [
   &InitializeParticle(
    Color=<rgba 15 15 15 1>
    Lifetime Max=0.5
    Lifetime Min=0.15
    Lifetime Mode=::NewEnumerator1
    Sprite Rotation Mode=::NewEnumerator2
    Sprite Size Mode=::NewEnumerator1
    Uniform Sprite Size Max=7
    Uniform Sprite Size Min=4
    UsePositionOffset=false
    Write Material Random=true
   )
   &SampleParticlesFromOtherEmitter(Apply Sampled Alpha=false, Color Sampling=::NewEnumerator0, Transform Data Option=1)
   &AddVelocity(Velocity=&UniformRangedVector(Minimum=<15 15 5>, Maximum=<-15 -15 -10>))
  ]
  particleUpdateScript [
   &ParticleState()
   &BlendSpace("Position Blend (World <-> Local)":NiagaraFloat=&MakeCustomFloatFromBool(
     Bool=&SetBoolByFloatComparison(A=&VectorLength(Vector=$Engine.Owner.Velocity), B=$User.Vel)
     True Float=1
     False Float=0.3
     Translation Contribution:NiagaraFloat=1
     Rotation Contribution:NiagaraFloat=0
    ))
   /Niagara/Modules/Update/Forces/V2/CurlNoiseForce(
    Noise Frequency=&RandomRangeFloat(Minimum=0.25, Maximum=1)
    Noise Strength=&RandomRangeFloat(Minimum=120, Maximum=300)
    Pan Noise=true
    Pan Noise Field=&RandomRangeVector(Minimum=<0 0 0>, Maximum=<1 1 1>)
   )
   &GravityForce(Gravity=&RandomRangeVector(Minimum=<0 0 -200>, Maximum=<0 0 -250>))
   &Drag(Drag=&RandomRangeFloat(Minimum=3, Maximum=8), Use Linear Drag=true, Use Rotational Drag=false)
   &SolveForcesAndVelocity(Clamp Velocity=false, Limit Acceleration=false)
   &ScaleSpriteSize(
    Uniform Curve Scale=0.5
    Uniform Curve Sprite Scale=DataInterface() {propertyValues {Curve {keys [{interpMode=RCIM_Cubic, tangentMode=RCTM_User, time=0, value=1}, {interpMode=RCIM_Cubic, tangentMode=RCTM_User, time=1, value=0}]}}}
   )
   &ScaleColor(
    Scale Alpha=&FloatFromCurve(FloatCurve=DataInterface() {propertyValues {Curve {keys [{arriveTangent=-1.933, interpMode=RCIM_Cubic, leaveTangent=-1.933, tangentMode=RCTM_User, time=0, value=1}, <at 1 0>]}}})
    Scale RGB=&VectorFromFloat(Value=&FloatFromCurve(
      FloatCurve=DataInterface() {propertyValues {Curve {keys [{arriveTangent=-2.5197, interpMode=RCIM_Cubic, leaveTangent=-2.5197, tangentMode=RCTM_User, time=0, value=1}, <at 1 0>]}}}
      Scale Curve=&RandomRangeFloat(Minimum=8, Maximum=15)
     ))
    ScaleA=false
    ScaleRGB=true
   )
  ]
  rendererClasses [&NiagaraSpriteRendererProperties]
  renderers [{rendererClass=&NiagaraSpriteRendererProperties}]
 }
}
summary {
 systemName=NS_Emote_ChefTartar_Firework1
 emitters [
  {emitterName=NE_TrailEmbers, rendererClasses [&NiagaraSpriteRendererProperties]}
  {emitterName=NE_RibbonSmoke, rendererClasses [&NiagaraRibbonRendererProperties]}
  {emitterName=NE_Spark, rendererClasses [&NiagaraSpriteRendererProperties]}
  {
   emitterName=NE_FireworkFlare_Parent
   rendererClasses [&NiagaraSpriteRendererProperties, &NiagaraLightRendererProperties]
  }
  {emitterName=NE_Fireworks_Mesh, rendererClasses [&NiagaraMeshRendererProperties]}
  {emitterName=NE_Flare, rendererClasses [&NiagaraLightRendererProperties, &NiagaraSpriteRendererProperties]}
  {emitterName=NE_Smoke, rendererClasses [&NiagaraSpriteRendererProperties]}
 ]
 userVariables [
  User.bIsFrontEnd {type=NiagaraBool}
  User.bIsFrontEndPreview {type=NiagaraBool}
  User.DepthFade {type=NiagaraFloat, defaultValue=0, description=""}
  User.Float {type=NiagaraFloat, defaultValue=0, description=""}
  User.Max {type=NiagaraFloat, defaultValue=250, description=""}
  User.Min {type=NiagaraFloat, defaultValue=90, description=""}
  User.offset {type=NiagaraFloat, defaultValue=5, description=""}
  User.Vel {type=NiagaraFloat, defaultValue=350, description=""}
 ]
}
